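# Physics asset providing collision shapes for the cape on the male medium BlackMonday Kansas player character cosmetic.
# Capsules approximate the pelvis, spine, clavicles, neck, legs and feet, with a sphere for the head.
# The spine_02 body adds two longer side-by-side capsules alongside its central one.
# Upper and lower arm bodies carry no shapes, and no constraints exist, so the asset defines collider volumes rather than a jointed ragdoll.
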PhysicsAsset /BRCosmetics/Characters/Player/Male/Medium/Bodies/M_MED_BlackMonday_Kansas/Meshes/M_MED_BlackMonday_Kansas_CapeColliders
bodies {
 calf_l [{shapeType=Capsule, length=31.242, radius=11.9116, center=<-19.4843 -0.5441 -0.2294>, rotation=<rot 0 94.0793 -89.4659>}]
 calf_r [{shapeType=Capsule, length=31.5362, radius=11.5971, center=<19.4486 0.3664 0.2406>, rotation=<rot 0 93.6446 -89.3731>}]
 clavicle_l [{shapeType=Capsule, length=8.7636, radius=10.1427, center=<9.9749 -0.9165 -1.595>, rotation=<rot 75.9352 -0.0001 175.5622>}]
 clavicle_r [{shapeType=Capsule, length=8.7636, radius=10.1427, center=<-9.9198 0.1841 1.5812>, rotation=<rot 75.9351 -0.0001 -4.4378>}]
 foot_l [{shapeType=Capsule, length=14.4367, radius=8.1146, center=<-4.2025 6.2288 -0.0237>, rotation=<rot -1.7928 -172.4181 -80.9982>}]
 foot_r [{shapeType=Capsule, length=14.4367, radius=8.2594, center=<4.3963 -6.0338 0.0727>, rotation=<rot -1.8067 -174.7364 -80.9479>}]
 head [{shapeType=Sphere, radius=17.3125, center=<11.028 -0.174 0>}]
 lowerarm_l []
 lowerarm_r []
 neck_01 [{shapeType=Capsule, length=6.005, radius=10.255, center=<0 2.6796 0.0003>, rotation=<rot 90 0 0>}]
 neck_02 [{shapeType=Capsule, length=13.4375, radius=7.3175, center=<-0.2642 -0.3408 0>, rotation=<rot 90 -171.1193 -175.3978>}]
 pelvis [{shapeType=Capsule, length=8.6562, radius=13.5388, center=<-2.6215 -1.3695 0>}]
 spine_01 [{shapeType=Capsule, length=10.7187, radius=12.4138, center=<-2.6215 -0.7156 0>}]
 spine_02 [
  {shapeType=Capsule, length=10.5937, radius=10.945, center=<-2.6215 -0.8003 0.0001>}
  {shapeType=Capsule, length=29.875, radius=11.1013, center=<2.5024 0.4992 3.9825>, rotation=<rot 90 0.0049 0.0049>}
  {shapeType=Capsule, length=29.875, radius=11.1013, center=<2.4704 0.4991 -3.9825>, rotation=<rot 90 -0.0396 -0.0396>}
 ]
 spine_03 [{shapeType=Capsule, length=9.1562, radius=10.7575, center=<-1.953 -0.8675 0.0001>}]
 spine_04 [{shapeType=Capsule, length=14.6562, radius=13.6325, center=<2.1467 0.0285 0.0002>}]
 spine_05 [{shapeType=Capsule, length=22.2812, radius=14.5625, center=<0.7875 2.3314 0>}]
 thigh_l [{shapeType=Capsule, length=25.252, radius=11.3318, center=<-19.839 -0.0002 0>, rotation=<rot 90 0 0>}]
 thigh_r [{shapeType=Capsule, length=24.8653, radius=11.1815, center=<19.0128 0 0>, rotation=<rot 90 0 0>}]
 upperarm_l []
 upperarm_r []
}
constraints []
